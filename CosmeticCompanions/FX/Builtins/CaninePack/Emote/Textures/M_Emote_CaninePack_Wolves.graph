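# Translucent, two-sided Niagara mesh-particle material that draws animated wolves for the CaninePack emote.
# Vertex offset math sizes the mesh from UVs along derived axes, optionally blending to SplineThicken.
# Two time-panned waves with separate offsets add front-back and up-down undulation, the latter masked by a sine along U.
# The offset, rotatable wolf texture is tinted by particle color, with opacity faded by particle alpha and view angle.
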Material /CosmeticCompanions/FX/Builtins/CaninePack/Emote/Textures/M_Emote_CaninePack_Wolves
graph {
 properties {
  bAutomaticallySetUsageInEditor=false
  bUsedWithNiagaraMeshParticles=true
  blendMode=BLEND_Translucent
  twoSided=true
  usedShadingModels=Default Lit
 }
 nodes [
  n1 = VertexNormalWS()
  n2 = Transform(None=Constant3Vector() {constant=<rgba 0 1 0 0>})
  n3 = CrossProduct(A=@n1, B=@n2)
  n4 = TextureCoordinate()
  n5 = MaterialFunctionCall(Float2=Clamp(None=ConstantBiasScale(None=@n4) {bias=-0.5, scale=2}) {minDefault=-1}) {materialFunction=&BreakOutFloat2Components}
  n6 = VectorParameter() {parameterName="Size&&SineWave", channelNames {a=UpDown, b=FrontBack, g=Width, r=Length}, defaultValue=<rgba 40 12.5 3 5>}
  n7 = Multiply(A=Multiply(A=Normalize(VectorInput=@n3), B=@n5.R) {constB=50}, B=@n6.Length) {constB=100}
  n8 = Multiply(A=Multiply(A=@n2, B=@n5.G) {constB=50}, B=@n6.Width) {constB=100}
  n9 = VectorParameter() {parameterName=Spline, defaultValue=<rgba 5 5 0 0>}
  n10 = StaticBool()
  n11 = ComponentMask(None=TextureCoordinate()) {r=true}
  n12 = VectorParameter() {parameterName="UpDown&&FrontBack", channelNames {a=Pan, b=Tiling, g=Pan, r=Tiling}, defaultValue=<rgba 1 0.1 1 0.05>}
  n13 = Time()
  n14 = Add(A=Add(A=Multiply(A=@n11, B=@n12.Tiling), B=Multiply(A=@n12.Pan, B=@n13)), B=ScalarParameter() {parameterName=PanOffset1})
  n15 = Multiply(
   A=LinearInterpolate(A=@n1, B=@n2, Alpha=ScalarParameter() {parameterName=Use Bottom})
   B=LinearInterpolate(A=Cosine(None=@n14), B=Sine(None=@n14), Alpha=ScalarParameter() {parameterName=Use Spline})
  )
  n16 = ScalarParameter() {parameterName=Use Bottom}
  n17 = Add(A=Add(A=Multiply(A=@n11, B=@n12.Tiling), B=Multiply(A=@n12.Pan, B=@n13)), B=ScalarParameter() {parameterName=PanOffset2})
  n18 = Multiply(A=@n6.UpDown, B=Multiply(A=LinearInterpolate(A=@n2, B=@n1, Alpha=@n16), B=Sine(None=@n17))) {constB=20}
  n19 = TextureCoordinate() {uTiling=0.5, vTiling=0.5}
  n20 = Saturate(None=Sine(None=ComponentMask(None=@n19) {r=true}))
  Add(
   A=LinearInterpolate(
    A=Add(A=@n7, B=@n8)
    B=MaterialFunctionCall(WidthBase=@n9.R, WidthTip=@n9.R, Expand U or V UV Channel=@n10) {materialFunction=&SplineThicken}.WorldPositionOffset
    Alpha=ScalarParameter() {parameterName=Use Spline}
   )
   B=Add(A=Multiply(A=@n6.FrontBack, B=@n15) {constB=20}, B=Multiply(A=@n18, B=@n20) {constB=20})
  )
  n21 = TextureCoordinate() {vTiling=-1}
  n22 = VectorParameter() {parameterName=WolfUV, defaultValue=<rgba 0 1 0 1>}
  n23 = ComponentMask(None=@n22.RGB) {g=true, r=true}
  n24 = TextureSampleParameter2D(UVs=MaterialFunctionCall(UVs=Add(A=@n21, B=@n23), "Rotation Angle (0-1)"=@n22.B) {materialFunction=&CustomRotator}.Rotated Values) {
   parameterName=Param
   samplerSource=SSM_Clamp_WorldGroupSettings
   texture=&/T_CaninePack_CenterWolfV2
  }
  n25 = ParticleColor()
  Multiply(A=@n24.RGB, B=@n25.RGB)
  n26 = Sine(None=ComponentMask(None=TextureCoordinate() {uTiling=0.5, vTiling=0.5}) {r=true})
  n27 = ScalarParameter() {parameterName=EdgeWidth}
  Multiply(A=Saturate(None=Add(A=@n26, B=@n27) {constB=0.5}), B=@n8)
  Multiply(
   A=Multiply(A=@n24.A, B=@n25.A)
   B=MaterialFunctionCall("FadeSensitivity (optional)"=ScalarParameter() {defaultValue=3, parameterName=AngleFade}) {materialFunction=&FadeBasedOnViewAngle}.Result
  )
  Saturate(None=@n5.G)
 ]
}
